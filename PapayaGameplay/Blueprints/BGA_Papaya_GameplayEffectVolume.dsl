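; Papaya gameplay volume actor that grants a configurable GameplayEffect to player pawns while they are inside its box.
; The construction script sizes the Volume box from VolumeBoxExtents.
; Begin overlap casts to FortPlayerPawn and applies the effect to a valid ability system component.
; End overlap removes all stacks by effect class rather than by stored handle, so overlapping volumes sharing an effect can clear each other's.
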(event OnComponentBeginOverlap(Volume)
  (params (OverlappedComponent :type Primitive%20Component%20Object%20Reference) (OtherActor :type Actor%20Object%20Reference) (OtherComp :type Primitive%20Component%20Object%20Reference) (OtherBodyIndex :type Integer) (bFromSweep :type Boolean) (SweepResult :type Hit%20Result%20Structure%20%28by%20ref%29))
  (bind _asfort_player_pawn (Utilities|Casting|CastToFortPlayerPawn OtherActor))
  (Utilities|IsValid (Ability|GetFortAbilitySystemComponent _asfort_player_pawn)
    (:Is%20Valid
      (GameplayEffects|ApplyGameplayEffectToSelf :self (Ability|GetFortAbilitySystemComponent _asfort_player_pawn) :GameplayEffectClass (Variables|Default|GetGameplayEffect) :Level 0.0))))

(event OnComponentEndOverlap(Volume)
  (params (OverlappedComponent :type Primitive%20Component%20Object%20Reference) (OtherActor :type Actor%20Object%20Reference) (OtherComp :type Primitive%20Component%20Object%20Reference) (OtherBodyIndex :type Integer))
  (bind _asfort_player_pawn (Utilities|Casting|CastToFortPlayerPawn OtherActor))
  (Utilities|IsValid (Ability|GetFortAbilitySystemComponent _asfort_player_pawn)
    (:Is%20Valid
      (GameplayEffects|RemoveActiveGameplayEffectbySourceEffect (Ability|GetFortAbilitySystemComponent _asfort_player_pawn) (Variables|Default|GetGameplayEffect) 0 -1))))

(fn ConstructionScript
  (Components|Box|SetBoxExtent (Variables|Default|GetVolume) (Variables|Default|GetVolumeBoxExtents) true))
