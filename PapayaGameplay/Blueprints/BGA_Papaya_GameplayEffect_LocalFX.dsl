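; Box volume actor for Papaya gameplay that applies a configured gameplay effect to player pawns inside it and removes it when they leave.
; Begin overlap retries after a short delay if the ability system is missing, as can happen when players spawn inside on the server.
; A dynamic effect list is also applied and removed on overlap.
; Custom events edit that list and immediately update every overlapping pawn.
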(event OnComponentBeginOverlap(Volume)
  (params (OverlappedComponent :type Primitive%20Component%20Object%20Reference) (OtherActor :type Actor%20Object%20Reference) (OtherComp :type Primitive%20Component%20Object%20Reference) (OtherBodyIndex :type Integer) (bFromSweep :type Boolean) (SweepResult :type Hit%20Result%20Structure%20%28by%20ref%29))
  (bind _asfort_player_pawn (Utilities|Casting|CastToFortPlayerPawn OtherActor))
  (Utilities|IsValid (Ability|GetFortAbilitySystemComponent _asfort_player_pawn)
    (:Is%20Valid
      _)
    (:Is%20Not%20Valid
      ;[1 1 1] If the player spawns into this volume on the server they might not have the ability component yet, a single frame delay appears to fix this
      (Utilities|FlowControl|Delay 0.1)
      (Utilities|IsValid (Ability|GetFortAbilitySystemComponent _asfort_player_pawn)
        (:Is%20Valid
          _)
        (:Is%20Not%20Valid
          (Development|PrintString "Papaya Gameplay Effect Local FX found INVALID Ability System")))))
  (GameplayEffects|ApplyGameplayEffectToSelf :self (Ability|GetFortAbilitySystemComponent _asfort_player_pawn) :GameplayEffectClass (Variables|Default|GetGameplayEffect) :Level 0.0)
  (for _array_element (Variables|Default|GetDynamicMeshGameplayEffects)
    (GameplayEffects|ApplyGameplayEffectToSelf :self (Ability|GetFortAbilitySystemComponent _asfort_player_pawn) :GameplayEffectClass _array_element :Level 0.0)))

(event OnComponentEndOverlap(Volume)
  (params (OverlappedComponent :type Primitive%20Component%20Object%20Reference) (OtherActor :type Actor%20Object%20Reference) (OtherComp :type Primitive%20Component%20Object%20Reference) (OtherBodyIndex :type Integer))
  (bind _asfort_player_pawn (Utilities|Casting|CastToFortPlayerPawn OtherActor))
  (Utilities|IsValid (Ability|GetFortAbilitySystemComponent _asfort_player_pawn)
    (:Is%20Valid
      (GameplayEffects|RemoveActiveGameplayEffectbySourceEffect (Ability|GetFortAbilitySystemComponent _asfort_player_pawn) (Variables|Default|GetGameplayEffect) 0 -1)
      (for _array_element (Variables|Default|GetDynamicMeshGameplayEffects)
        (GameplayEffects|RemoveActiveGameplayEffectbySourceEffect (Ability|GetFortAbilitySystemComponent _asfort_player_pawn) _array_element 0 -1)))
    (:Is%20Not%20Valid
      (Development|PrintString "Papaya Gameplay Effect Local FX found INVALID Ability System"))))

(event Custom|HandleMeshGameplayEffect
  (params (Gameplay%20Effect :type Gameplay%20Effect%20Class%20Reference) (Add%20Effect :type Boolean))
  ;[1 1 1] Adding and removing gameplay effects to all players in this volume
  (if Add%20Effect
    (CallFunction|AddDynamicMeshGameplayEffect Gameplay%20Effect)
    (else
      (CallFunction|RemoveDynamicMeshGameplayEffect Gameplay%20Effect))))

(event Custom|Add%20Dynamic%20Mesh%20Gameplay%20Effect
  (params (Add%20Gameplay%20Effect :type Gameplay%20Effect%20Class%20Reference))
  ;[1 1 1] Adding and removing gameplay effects to all players in this volume
  (if (not (Utilities|Array|ContainsItem (Variables|Default|GetDynamicMeshGameplayEffects) Add%20Gameplay%20Effect))
    (Utilities|Array|Add (Variables|Default|GetDynamicMeshGameplayEffects) Add%20Gameplay%20Effect)
    (for _array_element (Collision|GetOverlappingActors "/Script/FortniteGame.FortPlayerPawn")
      (Utilities|IsValid (Ability|GetAbilitySystemComponent _array_element)
        (:Is%20Valid
          (GameplayEffects|ApplyGameplayEffectToSelf :self (Ability|GetAbilitySystemComponent _array_element) :GameplayEffectClass Add%20Gameplay%20Effect :Level 0.0))))))

(event Custom|Remove%20Dynamic%20Mesh%20Gameplay%20Effect
  (params (Remove%20Gameplay%20Effect :type Gameplay%20Effect%20Class%20Reference))
  ;[1 1 1] Adding and removing gameplay effects to all players in this volume
  (if (Utilities|Array|ContainsItem (Variables|Default|GetDynamicMeshGameplayEffects) Remove%20Gameplay%20Effect)
    (Utilities|Array|RemoveItem (Variables|Default|GetDynamicMeshGameplayEffects) Remove%20Gameplay%20Effect)
    (for _array_element (Collision|GetOverlappingActors "/Script/FortniteGame.FortPlayerPawn")
      (Utilities|IsValid (Ability|GetAbilitySystemComponent _array_element)
        (:Is%20Valid
          (GameplayEffects|RemoveActiveGameplayEffectbySourceEffect (Ability|GetAbilitySystemComponent _array_element) Remove%20Gameplay%20Effect 0 -1))))))

(fn ConstructionScript
  (Components|Box|SetBoxExtent (Variables|Default|GetVolume) (Variables|Default|GetVolumeBoxExtents) true))
