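; Searchable barrel container for Papaya that a mesh event can disable, making it look empty.
; Persistent-on saves the searched state, marks the barrel searched and flags it disabled.
; Persistent-off clears both only if it was unsearched beforehand; otherwise both stay set.
; A loot reset while disabled starts a short timer that re-applies the searched state, covering disables that arrive during initialization.
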(event EventBeginPlay
  (|CallParentFunction) ; This node is disabled and will not be called.\nDrag off pins to build functionality.
)

(event Collision|EventActorBeginOverlap
  (params (OtherActor :type Actor%20Object%20Reference)))

(event EventTick
  (params (DeltaSeconds :type Float%20%28single-precision%29)))

(event OnMeshNetEvent(FortEventResponder)
  (params (EventTag :type Gameplay%20Tag%20Structure) (bPersistentOnOff :type Boolean) (EventType :type EEventResponderEventType%20Enum :default "OneShotEvent") (EventParameters :type String))
  ;[1 1 1] Enable/Disable on mesh event. When disabled, switches to the searched state and looks empty.
  (if bPersistentOnOff
    (Variables|Default|SetbWasAlreadySearchedWhenDisabled (Container|IsAlreadySearched))
    (Container|SetAlreadySearched true)
    (Variables|Default|SetbDisabled true)
    (elif (Variables|Default|GetbWasAlreadySearchedWhenDisabled)
      _
      (else
        (Container|SetAlreadySearched false)
        (Variables|Default|SetbDisabled false)))))

(event EventOnLootReset
  ;[1 1 1] If loot is reset while disabled (this can happen during initialization if a mesh event disables this object before it finishes initializing) switch back to the searched state.
  (if (Variables|Default|GetbDisabled)
    (Utilities|Time|SetTimerbyEvent (delegate Custom|LootResetWhileDisabled) 0.333)))

(event Custom|LootResetWhileDisabled
  ;[1 1 1] If loot is reset while disabled (this can happen during initialization if a mesh event disables this object before it finishes initializing) switch back to the searched state.
  (Container|SetAlreadySearched true))

(fn ConstructionScript
  (|CallParentFunction))
