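; Gameplay ability that makes the activating player pawn bounce, relaunching it each time it lands.
; Activation binds LaunchPawns to OnPawnLanded and also calls it at once if the pawn is not falling.
; LaunchPawns calls LaunchCharacterJump with BounceVelZ overriding vertical velocity while horizontal velocity is added.
; The horizontal term blends facing with clamped velocity, but its NotEqual gate's tolerance of 50 exceeds any possible difference, zeroing it.
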(event Ability|EventActivateAbility
  (bind _returnvalue (Ability|GetActivatingPawn))
  (EventDispatchers|BindEventtoOnPawnLanded _returnvalue (EventDispatchers|CreateEvent :!event Custom|LaunchPawns))
  (if (not (AI|Components|NavMovement|IsFalling (Variables|/Script/Engine.Character|GetCharacterMovement _returnvalue)))
    (CallFunction|LaunchPawns (Collision|MakeHitResult false))))

(event Custom|LaunchPawns
  (params (Hit :type Hit%20Result%20Structure%20%28by%20ref%29))
  (bind (_velocity_x _velocity_y) (Math|Vector|BreakVector (Variables|/Script/Engine.MovementComponent|GetVelocity_X (Variables|/Script/Engine.Character|GetCharacterMovement (Ability|GetActivatingPawn)))))
  (bind _asfort_player_pawn (Utilities|Casting|CastToFortPlayerPawn (Ability|GetActivatingPawn)))
  (Pawn|Character|LaunchCharacterJump _asfort_player_pawn (Math|Vector|MakeVector (* (Math|Conversions|ToInteger(Boolean) (Math|Vector|NotEqual(Vector) (Transformation|GetActorForwardVector (Ability|GetActivatingPawn)) (Math|Vector|vector+vector (Math|Vector|MakeVector (.x (Transformation|GetActorForwardVector (Ability|GetActivatingPawn))) (.y (Transformation|GetActorForwardVector (Ability|GetActivatingPawn)))) (Math|Vector|MakeVector (Math|Float|MapRangeClamped _velocity_x -6e+02 6e+02 -1.0 1.0) (Math|Float|MapRangeClamped _velocity_y -6e+02 6e+02 -1.0 1.0))) 5e+01)) (.x (Math|Vector|vector+vector (Math|Vector|MakeVector (.x (Transformation|GetActorForwardVector (Ability|GetActivatingPawn))) (.y (Transformation|GetActorForwardVector (Ability|GetActivatingPawn)))) (Math|Vector|MakeVector (Math|Float|MapRangeClamped _velocity_x -6e+02 6e+02 -1.0 1.0) (Math|Float|MapRangeClamped _velocity_y -6e+02 6e+02 -1.0 1.0))))) (* (Math|Conversions|ToInteger(Boolean) (Math|Vector|NotEqual(Vector) (Transformation|GetActorForwardVector (Ability|GetActivatingPawn)) (Math|Vector|vector+vector (Math|Vector|MakeVector (.x (Transformation|GetActorForwardVector (Ability|GetActivatingPawn))) (.y (Transformation|GetActorForwardVector (Ability|GetActivatingPawn)))) (Math|Vector|MakeVector (Math|Float|MapRangeClamped _velocity_x -6e+02 6e+02 -1.0 1.0) (Math|Float|MapRangeClamped _velocity_y -6e+02 6e+02 -1.0 1.0))) 5e+01)) (.y (Math|Vector|vector+vector (Math|Vector|MakeVector (.x (Transformation|GetActorForwardVector (Ability|GetActivatingPawn))) (.y (Transformation|GetActorForwardVector (Ability|GetActivatingPawn)))) (Math|Vector|MakeVector (Math|Float|MapRangeClamped _velocity_x -6e+02 6e+02 -1.0 1.0) (Math|Float|MapRangeClamped _velocity_y -6e+02 6e+02 -1.0 1.0))))) (Variables|Default|GetBounceVelZ)) false true true true))
