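; Spotlight child Blueprint for Papaya Donut that can start off and be switched on or off at runtime.
; On BeginPlay it applies its default-off setting, since the parent construction script sets the light up as on.
; TurnOffSpotlight acts only on state changes, toggling light shaft visibility and the material's "Rotation Max" to stop or resume rotation.
; Mesh network events tagged Spotlights.ForceOn or ForceOff drive that toggle, inverting it on PersistentEventEnd.
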(event EventBeginPlay
  (|CallParentFunction)
  (CallFunction|TurnOffSpotlight (Variables|Default|GetSpotlight%20Default%20Off)) ; we want to support off by default, the parent class sets things up in the construction scrip as on by default
)

(event Custom|TurnOffSpotlight
  (params (Off :type Boolean))
  ;[1 1 1] Support Turning off and back on properly
  (if (not (Math|Boolean|Equal(Boolean) Off (Variables|Default|GetCurrent%20Off)))
    (bind _output_get (Variables|Default|SetCurrent%20Off Off))
    ;;[1 1 1] Hide the light when off
    (Rendering|SetVisibility (Variables|Default|Getsm_neo_tilted_light_shaft) (not _output_get) false)
    ;;[1 1 1] Stop or start the mesh rotation
    (Rendering|Material|SetScalarParameterValueonMaterials (Variables|Default|GetS_NeoTilted_Spotlight_01) "Rotation Max" (Math|Float|SelectFloat 0.0 (Variables|SpotlightFXSettings|GetRotation%20Max) (Variables|Default|GetCurrent%20Off)))))

(event OnMeshNetEvent(FortEventResponder)
  (params (EventTag :type Gameplay%20Tag%20Structure) (bPersistentOnOff :type Boolean) (EventType :type EEventResponderEventType%20Enum :default "OneShotEvent") (EventParameters :type String))
  (bind _returnvalue (Utilities|Enum|NotEqual(Enum) EventType "PersistentEventEnd"))
  ;[0.2705 1 0.9801] Mesh Network Support to dynamically turn on / off with these commands - cheat testing "meshcheat MeshNetEventStart Spotlights.ForceOn"
  (switch Utilities|FlowControl|Switch|SwitchonGameplayTag EventTag
    (:Spotlights.ForceOn
      (CallFunction|TurnOffSpotlight (not _returnvalue)))
    (:Spotlights.ForceOff
      (CallFunction|TurnOffSpotlight _returnvalue))))

(fn ConstructionScript
  (|CallParentFunction))
